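# Translucent unlit master material for anime-style vehicle glass, with lighting computed in-shader and written to emissive.
# Lit and shaded colors blend by a light-direction dot product through a curve atlas or banded cel function.
# Stylized specular uses Phong or Blinn highlights, or switchable planar, cubemap and MatCap textures with optional masking.
# Fresnel helps drive opacity, and the attributes pass through character and vehicle PostFX functions.
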Material /VehicleCosmetics/Materials/VehicleParents_Version2/MAT_Vehicle_AnimeGlassShader
graph {
 properties {
  bAutomaticallySetUsageInEditor=false
  bTangentSpaceNormal=false
  bUseMaterialAttributes=true
  bUsedWithClothing=true
  bUsedWithMeshDeformer=true
  bUsedWithMorphTargets=true
  bUsedWithSkeletalMesh=true
  blendMode=BLEND_Translucent
  shadingModel=MSM_Unlit
  usedShadingModels=Unlit
 }
 nodes [
  Composite() {subgraphName=Construct Base Inputs}
  n1 = LinearInterpolate(A=NamedRerouteUsage(), B=ComponentMask(None=NamedRerouteUsage()) {b=true, g=true, r=true}, Alpha=NamedRerouteUsage())
  n2 = NamedRerouteUsage()
  n3 = NamedRerouteUsage()
  n4 = Constant()
  MaterialExpressionReroute_4 = Reroute(None=@n4)
  n5 = PreSkinnedPosition()
  n6 = ShaderStageSwitch(PixelShader=VertexInterpolator(VS=@n5).PS, VertexShader=@n5)
  MaterialExpressionReroute_5 = Reroute(None=@n6)
  n7 = MaterialFunctionCall!(Material) {materialFunction=/Game/Packages/Fortress_SharedMaterials/Base_Material_Functions/CharacterBrightnessandScale_Balance}
  MaterialExpressionReroute_6 = Reroute(None=@n7.Distance Blend)
  n8 = Fresnel() {baseReflectFraction=0, exponent=1}
  MaterialExpressionReroute_7 = Reroute(None=@n8)
  MaterialFunctionCall!(In=MaterialFunctionCall(
    Material=SetMaterialAttributes!(MaterialAttributes, Emissive Color=@n1, Opacity=@n2, Normal=@n3) {attributeSetTypes [#1, #2, #3]}
    M=@MaterialExpressionReroute_4.M
    Pre Skinned Local Position=@MaterialExpressionReroute_5.Pre Skinned Local Position
    Distance Blend=@MaterialExpressionReroute_6.Distance Blend
    Fresnel=@MaterialExpressionReroute_7.Fresnel
   ) {materialFunction=&MF_CharacterParent_2023_PostFX}.Result) {materialFunction=/VehicleCosmetics/Materials/MaterialFunctions/MF_Vehicle_FNPostFX}
  n9 = MaterialFunctionCall!(CustomUV) {materialFunction=/Game/Global/Materials/Character_Parent_2023/Anime_Parent_2023/MaterialFunctions/MFC_AnimeShader_2023_Pre}
  NamedRerouteDeclaration(None=@n9.Normal) {name=NormalMap, nodeColor=<rgba 0 0.0156 1 1>}
  n10 = NamedRerouteUsage()
  n11 = NamedRerouteUsage()
  n12 = ScalarParameter() {group=02_Anime Parameters, parameterName=SpecularHighlight_LockedZVector, sliderMax=1}
  n13 = AppendVector(A=Constant2Vector() {g=1, r=1}, B=Saturate(None=OneMinus(None=@n12)))
  n14 = Multiply(A=Add(A=CameraVectorWS(), B=@n11), B=@n13)
  n15 = NamedRerouteDeclaration(None=Transform(None=NamedRerouteUsage())) {name=N_WS, nodeColor=<rgba 0.3125 0 1 1>}
  n16 = Multiply(A=@n11, B=@n13)
  n17 = StaticSwitchParameter(
   True=DotProduct(A=Normalize(VectorInput=@n14), B=@n15) {desc=Blinn}
   False=DotProduct(A=Normalize(VectorInput=@n16), B=ReflectionVectorWS(CustomWorldNormal=@n15)) {desc=Phong}
  ) {desc=Default is the Phong Highlight, group=02_Anime Parameters, parameterName=UseBlinnHighlight, sortPriority=0}
  n18 = NamedRerouteUsage()
  n19 = ScalarParameter() {group=02_Anime Parameters, parameterName=SpecularHighlight_CameraFacing_Blend, sliderMax=1}
  n20 = SphereMask(
   A=Saturate(None=LinearInterpolate(A=@n17, B=@n18, Alpha=@n19))
   B=Constant() {r=1}
   Radius=ScalarParameter() {defaultValue=0.1, group=02_Anime Parameters, parameterName=SpecularHighlight_Size, sliderMax=1, sliderMin=0.0001}
   Hardness=ScalarParameter() {defaultValue=0.5, group=02_Anime Parameters, parameterName=SpecularHighlight_Hardness, sliderMax=1}
  ) {hardnessPercent=78}
  n21 = StaticSwitchParameter(
   True=NamedRerouteUsage()
   False=StaticSwitchParameter(
    True=NamedRerouteUsage()
    False=StaticSwitchParameter(True=@n10, False=@n20) {group=02_Anime Parameters, parameterName=UsePlanarSpecular, sortPriority=1}
   ) {group=02_Anime Parameters, parameterName=UseCubeMapSpecular, sortPriority=1}
  ) {group=02_Anime Parameters, parameterName=UseMatCapSpecular, sortPriority=1}
  n22 = ChannelMaskParameter(None=NamedRerouteUsage()) {
   desc="Mask to blend between the specular highlight and the Planar, Cube or MatCap functions."
   group=02_Anime Parameters
   parameterName=FXChannel_SpecularMask
   sortPriority=10
  }
  n23 = LinearInterpolate(A=@n20, B=@n21, Alpha=@n22.R)
  NamedRerouteDeclaration(None=Multiply(
    A=StaticSwitchParameter(True=@n23, False=@n21) {group=02_Anime Parameters, parameterName=UseSpecularFXMasking, sortPriority=1}
    B=ScalarParameter() {defaultValue=1, group=02_Anime Parameters, parameterName=SpecularHighlight_Intensity, sliderMax=8}
   )) {name=Specular_Curve, nodeColor=<rgba 1 0.4453 0 1>}
  NamedRerouteDeclaration(None=@n9.FX) {name=Map_FX, nodeColor=<rgba 0 0.326 1 1>}
  n24 = TextureCoordinate()
  n25 = TextureSampleParameter2D!(
   UVs=MaterialFunctionCall!(Vector=ReflectionVectorWS()) {materialFunction=/Engine/Functions/Engine_MaterialFunctions01/Coordinates/LongLatToUV}.Result
   "DDX(UVs)"=DDX(Value=@n24)
   "DDY(UVs)"=DDY(Value=@n24)
   Apply View MipBias
  ) {
   group=01_Anime Maps
   mipValueMode=TMVM_Derivative
   parameterName=SpecularMap_Cube
   samplerSource=SSM_Wrap_WorldGroupSettings
   sortPriority=20
   texture=/Game/Global/Materials/Character_Parent_2023/Anime_Parent_2023/Textures/T_CubeMap_Highlight
  }
  NamedRerouteDeclaration(None=@n25.RGB) {name=CubeSpec, nodeColor=<rgba 1 0.4327 0 1>}
  n26 = Transform(None=Constant3Vector() {constant=<rgba 0 0 1 0>}) {transformSourceType=TRANSFORMSOURCE_View, transformType=TRANSFORM_Local}
  n27 = MaterialFunctionCall(Float2=MaterialFunctionCall(Vector or UVs=ComponentMask(None=@n26) {g=true, r=true}) {materialFunction=&VectorToRadialValue}.Radial Coordinates) {materialFunction=&BreakOutFloat2Components}
  n28 = AppendVector(A=Add(A=Multiply(A=@n27.R) {constB=4}) {constB=-1}, B=Multiply(A=ComponentMask(None=@n26) {b=true}) {constB=-1})
  n29 = WorldPosition()
  n30 = MaterialFunctionCall(Float3=Divide(
    A=TransformPosition(None=@n29.XYZ) {transformSourceType=TRANSFORMPOSSOURCE_World}
    B=ScalarParameter() {defaultValue=30, group=02_Anime Parameters, parameterName=SpecularHighlight_Planar_Scale, sliderMax=200}
   ) {constB=30}) {materialFunction=&BreakOutFloat3Components}
  n31 = MaterialFunctionCall(X=@n30.R, Y=@n30.R) {materialFunction=&MakeFloat2}
  n32 = MaterialFunctionCall(X=@n30.G, Y=@n30.G) {materialFunction=&MakeFloat2}
  n33 = Add(A=Multiply(A=StaticSwitchParameter(True=@n31.Result, False=@n32.Result) {defaultValue=true, group=02_Anime Parameters, parameterName=UsePlanarReflectionYPlane, sortPriority=2}) {constB=-1}) {constB=0.5}
  n34 = TextureSampleParameter2D!(
   UVs=Rotator(
    Coordinate=Add(A=@n28, B=@n33)
    Time=ScalarParameter() {group=02_Anime Parameters, parameterName=SpecularHighlight_Planar_Rotate, sliderMax=1, sliderMin=-1}
   ) {speed=3.1415}
   Apply View MipBias
  ) {
   group=01_Anime Maps
   parameterName=SpecularMap_Planar
   samplerSource=SSM_Wrap_WorldGroupSettings
   sortPriority=20
   texture=/Game/Global/Materials/Character_Parent_2023/Anime_Parent_2023/Textures/T_Planar_Highlight
  }
  NamedRerouteDeclaration(None=@n34.RGB) {name=PlanarSpec, nodeColor=<rgba 1 0.4327 0 1>}
  n35 = TransformPosition(None=WorldPosition().XYZ) {transformSourceType=TRANSFORMPOSSOURCE_World, transformType=TRANSFORMPOSSOURCE_Camera}
  n36 = Normalize(VectorInput=VertexNormalWS())
  n37 = MaterialFunctionCall(XY=CrossProduct(A=Normalize(VectorInput=@n35), B=Transform(None=@n36) {transformSourceType=TRANSFORMSOURCE_World, transformType=TRANSFORM_Camera})) {materialFunction=&Swizzle}
  n38 = TextureSampleParameter2D(UVs=Add(A=Multiply(A=@n37.YX) {constB=0.5}) {constB=0.5}) {
   group=01_Anime Maps
   parameterName=SpecularMap_MatCap
   samplerSource=SSM_Wrap_WorldGroupSettings
   sortPriority=20
   texture=/Game/Global/Materials/Character_Parent_2023/Anime_Parent_2023/Textures/T_MatCap_Highlight
  }
  NamedRerouteDeclaration(None=@n38.RGB) {name=MatCapSpec, nodeColor=<rgba 1 0.4327 0 1>}
  NamedRerouteDeclaration(None=@n9.Specular_Color) {name=SpecularColor, nodeColor=<rgba 0.3359 1 0 1>}
  NamedRerouteDeclaration(None=DotProduct(A=PixelNormalWS(), B=CameraVectorWS())) {name=NdotV, nodeColor=<rgba 1 0.6242 0 1>}
  NamedRerouteDeclaration(None=@n9.Color_Lit) {name=ColorLitMap, nodeColor=<rgba 1 0 0.6562 1>}
  NamedRerouteDeclaration(None=MaterialFunctionCall() {materialFunction=/Game/Global/Materials/Character_Parent_2023/Anime_Parent_2023/MaterialFunctions/MFC_AnimeShader_LightDirection}.Direction Vector) {name=LightVector, nodeColor=<rgba 1 0.934 0.5103 1>}
  NamedRerouteDeclaration(None=@n9.Color_Shaded) {name=ColorShadedMap, nodeColor=<rgba 0 0.9766 1 1>}
  n39 = Multiply(A=DotProduct(A=NamedRerouteUsage(), B=NamedRerouteUsage())) {constB=0.5}
  n40 = Saturate(None=Add(A=Add(A=@n39) {constB=0.5}, B=ScalarParameter() {group=02_Anime Parameters, parameterName=ShadingCurve_Bias, sliderMax=1, sliderMin=-1}))
  n41 = CurveAtlasRowParameter(CurveTime=@n40) {
   atlas=/Game/Characters/Shared/CurveAtlas/AthenaCharacterCurveAtlas
   curve=/Game/Characters/Shared/CurveAtlas/HB_TwoToneShadingCurve
   defaultValue=22
   group=02_Anime Parameters
   parameterName=ShadingCurve
   sortPriority=5
  }
  n42 = ScalarParameter() {defaultValue=0.02, group=02_Anime Parameters, parameterName=ShadingCurve_Width, sliderMax=1}
  n43 = ScalarParameter() {defaultValue=1, group=02_Anime Parameters, parameterName=ShadingCurve_Bands, sliderMax=4, sliderMin=1}
  n44 = ScalarParameter() {
   defaultValue=1
   desc="To convert the (mathmatical) linear greadient to linear perceived gradient"
   group=09_Anime Parameters_TechArt
   parameterName=ShadingCurve_Exp
   sliderMax=2.2
   sliderMin=0.4545
  }
  n45 = Power(
   Base=StaticSwitchParameter(
    True=ChannelMaskParameter(None=@n41) {group=02_Anime Parameters, parameterName=ShadingCurveChannel, sortPriority=10}.R
    False=MaterialFunctionCall!(NdL=@n40, Width=@n42, Bands Amount=@n43) {materialFunction=/Game/Global/Materials/Character_Parent_2023/Anime_Parent_2023/MaterialFunctions/MF_SmoothCelShadingWithBands}.Divide
   ) {group=02_Anime Parameters, parameterName=UseShadingCurve, sortPriority=-1}
   Exp=@n44
  ) {constExponent=2.2}
  NamedRerouteDeclaration(None=LinearInterpolate(A=NamedRerouteUsage(), B=NamedRerouteUsage(), Alpha=Saturate(None=@n45))) {name=ColorMixed, nodeColor=<rgba 0.0078 1 0 1>}
  n46 = ScalarParameter() {group=02_Anime Parameters, parameterName=Opacity_Min, sliderMax=1}
  n47 = ScalarParameter() {defaultValue=1, group=02_Anime Parameters, parameterName=Opacity_Max, sliderMax=1}
  n48 = Fresnel(
   ExponentIn=ScalarParameter() {defaultValue=3, group=02_Anime Parameters, parameterName=Opacity_FresnelExp, sliderMax=8}
   BaseReflectFractionIn=ScalarParameter() {defaultValue=0.1, group=02_Anime Parameters, parameterName=Opacity_FresnelBaseReflectionFrac, sliderMax=8}
  )
  NamedRerouteDeclaration(None=Add(A=NamedRerouteUsage(), B=LinearInterpolate(A=@n46, B=@n47, Alpha=@n48))) {name=Opacity, nodeColor=<rgba 1 0 0.3281 1>}
  PinBase!(M=@n4, Pre Skinned Local Position=@n6, Distance Blend=@n7.Distance Blend, Fresnel=@n8) {reroutePins [
    M {expression=&self:MaterialExpressionReroute_4}
    Pre Skinned Local Position {expression=&self:MaterialExpressionReroute_5}
    Distance Blend {expression=&self:MaterialExpressionReroute_6}
    Fresnel {expression=&self:MaterialExpressionReroute_7}
   ]}
  PinBase!()
 ]
}
