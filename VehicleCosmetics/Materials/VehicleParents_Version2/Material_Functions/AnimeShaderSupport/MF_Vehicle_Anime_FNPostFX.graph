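# Material function for anime-shaded vehicles that layers damage and de-rez transition effects onto incoming material attributes.
# Anime damage is the base, and the UsePBRSwitch static switch optionally blends in standard vehicle damage.
# That blend uses a hard 0/1 mask thresholded from the PBR_Shading scalar at 0.5.
# The input's shading model is reapplied to the result before it passes through the de-rez transition function.
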MaterialFunction /VehicleCosmetics/Materials/VehicleParents_Version2/Material_Functions/AnimeShaderSupport/MF_Vehicle_Anime_FNPostFX
graph {nodes [
  n1 = GetMaterialAttributes(None=FunctionInput() {inputType=FunctionInput_MaterialAttributes}) {attributeGetTypes [#1]}
  n2 = MaterialFunctionCall!(In=@n1.MaterialAttributes, "Clear Coat/Damage Mask (_M B)") {materialFunction=&/MF_AnimeVehicle_Damage}
  n3 = StaticSwitchParameter(
   True=BlendMaterialAttributes(
    A=@n2.Result
    B=MaterialFunctionCall(Material=@n1.MaterialAttributes) {materialFunction=&MF_BaseVehicle_Damage}.Result
    Alpha=NamedRerouteUsage()
   )
   False=@n2.Result
  ) {group=02_Anime Parameters, parameterName=UsePBRSwitch, sortPriority=0}
  FunctionOutput(None=MaterialFunctionCall(In=SetMaterialAttributes!(MaterialAttributes=@n3, ShadingModel=@n1.ShadingModel) {attributeSetTypes [#1]}) {materialFunction=&MF_VehicleDeRez_Transition}.Result)
  n4 = Subtract(A=ScalarParameter() {group=03_Anime_PBR_Override, parameterName=PBR_Shading, sliderMax=1, sortPriority=5}) {constB=0.5}
  NamedRerouteDeclaration(None=Saturate(None=Ceil(None=@n4))) {name=PBR_Switch, nodeColor=<rgba 0 1 0.8566 1>}
 ]}
