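# Material function that post-processes anime-shaded vehicle chassis, rebuilding emissive color for headlights, brake and reverse lights, boost and emissive parts.
# MF_Vehicle_ChassisAnimeVariables supplies masks, colors and vehicle-state values, combined with brightness parameters and optional base-color overrides.
# Mask-driven blending mixes in color adjusted by MF_TODColorAdjustment, with an optional PBR path.
# A UseChassisSettings static switch gates the emissive write to the material attributes.
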MaterialFunction /VehicleCosmetics/Materials/VehicleParents_Version2/Material_Functions/AnimeShaderSupport/MF_Vehicle_ChassisAnime_Post
graph {nodes [
  n1 = NamedRerouteUsage()
  n2 = NamedRerouteUsage()
  n3 = MaterialFunctionCall!(Color=@n2) {materialFunction=/Game/Packages/Fortress_SharedMaterials/Anime_Master_Material/MF_TODColorAdjustment}
  n4 = NamedRerouteUsage()
  n5 = NamedRerouteUsage()
  n6 = NamedRerouteUsage()
  n7 = NamedRerouteUsage()
  n8 = LinearInterpolate(
   A=NamedRerouteUsage()
   B=StaticSwitchParameter(True=LinearInterpolate(A=@n3.Result, B=@n2, Alpha=@n4), False=@n3.Result) {group=02_Anime Parameters, parameterName=UsePBRSwitch, sortPriority=0}
   Alpha=StaticSwitchParameter(True=LinearInterpolate(A=@n5, B=@n6, Alpha=@n7), False=@n5) {group=02_Anime Parameters, parameterName=UsePBRSwitch, sortPriority=0}
  )
  FunctionOutput(None=StaticSwitch(
    True=SetMaterialAttributes!(MaterialAttributes=@n1, EmissiveColor=@n8) {attributeSetTypes [#1]}
    False=NamedRerouteUsage()
    Value=NamedRerouteUsage()
   )) {sortPriority=15}
  NamedRerouteDeclaration(None=StaticBoolParameter() {group=Vehicle_ShaderControls, parameterName=UseChassisSettings, sortPriority=30}) {name=UseChassisBool, nodeColor=<rgba 0.6406 0 1 1>}
  n9 = Max(A=NamedRerouteUsage(), B=NamedRerouteUsage())
  n10 = NamedRerouteUsage()
  n11 = Max(A=Max(A=Max(A=@n9, B=@n10), B=NamedRerouteUsage()), B=NamedRerouteUsage())
  NamedRerouteDeclaration(None=Saturate(None=ConstantBiasScale(None=@n11) {bias=0, scale=2})) {name=EmissiveMasks, nodeColor=<rgba 0.4297 0 1 1>}
  n12 = GetMaterialAttributes(None=FunctionInput() {inputName=Input, inputType=FunctionInput_MaterialAttributes}) {attributeGetTypes [#2, #1]}
  NamedRerouteDeclaration(None=@n12.EmissiveColor) {name=EmissiveColorSource, nodeColor=<rgba 0.6172 1 0 1>}
  NamedRerouteDeclaration(None=@n12.MaterialAttributes) {name=Source_MA, nodeColor=<rgba 1 0 0.9375 1>}
  n13 = Max(A=Multiply(A=NamedRerouteUsage(), B=NamedRerouteUsage()), B=Multiply(A=NamedRerouteUsage(), B=NamedRerouteUsage()))
  n14 = Multiply(A=NamedRerouteUsage(), B=NamedRerouteUsage())
  n15 = ConstantBiasScale(None=Max(A=Max(A=@n13, B=@n14), B=NamedRerouteUsage())) {bias=0, scale=2}
  n16 = Multiply(A=NamedRerouteUsage(), B=Saturate(None=NamedRerouteUsage()))
  NamedRerouteDeclaration(None=Saturate(None=Max(A=@n15, B=@n16))) {name=EmissiveMasksStates, nodeColor=<rgba 0.4297 0 1 1>}
  n17 = MaterialFunctionCall() {materialFunction=&/MF_Vehicle_ChassisAnimeVariables}
  NamedRerouteDeclaration(None=@n17.Color_Headlight) {name=HeadlightColor, nodeColor=<rgba 0.9282 0 1 1>}
  n18 = Subtract(A=ScalarParameter() {group=03_Anime_PBR_Override, parameterName=PBR_Shading, sliderMax=1, sortPriority=5}) {constB=0.5}
  NamedRerouteDeclaration(None=Saturate(None=Ceil(None=@n18))) {name=PBR_Switch, nodeColor=<rgba 0 1 0.8566 1>}
  NamedRerouteDeclaration(None=@n17.Color_Boost) {name=BoostColor, nodeColor=<rgba 0.9282 0 1 1>}
  n19 = VectorParameter() {
   desc=Brightness Settings
   group=Vehicle_Chassis_Emissive
   parameterName=Chassis_EmissiveBrightness
   sortPriority=0
   channelNames {a=BoostEmissive, b=BrakeLightEmissive, g=HeadLightEmissive, r=ChassisEmissiveParts}
   defaultValue=<rgba 1 1 1 1>
  }
  NamedRerouteDeclaration(None=@n19.HeadLightEmissive) {name=HeadlightEmissiveBrightness, nodeColor=<rgba 0.0436 1 0 1>}
  NamedRerouteDeclaration(None=@n19.BrakeLightEmissive) {name=BrakelightEmissiveBrightness, nodeColor=<rgba 0.0436 1 0 1>}
  NamedRerouteDeclaration(None=@n19.BoostEmissive) {name=BoostEmissiveBrightness, nodeColor=<rgba 0.0436 1 0 1>}
  NamedRerouteDeclaration(None=@n17.BPVariable_BoostGlowIntensity) {name=BoostGlowVariable, nodeColor=<rgba 0 0 0 1>}
  NamedRerouteDeclaration(None=@n17.BPVariable_isBraking) {name=isBrakingVariable, nodeColor=<rgba 0 0 0 1>}
  NamedRerouteDeclaration(None=@n17.BPVariable_ReverseLight) {name=ReverseLightVariable, nodeColor=<rgba 0 0 0 1>}
  NamedRerouteDeclaration(None=@n17.BPVariable_HeadlightsFullyOn) {name=HeadlightsFullyOnVariable, nodeColor=<rgba 0 0 0 1>}
  NamedRerouteDeclaration(None=@n17.Masks_Boost) {name=ChassisBoostEmissiveMask, nodeColor=<rgba 0 0.1726 0.4792 1>}
  NamedRerouteDeclaration(None=@n17.EmissiveMasks_Headlight) {name=HeadlightEmissiveMask, nodeColor=<rgba 0 0.1726 0.4792 1>}
  NamedRerouteDeclaration(None=@n17.EmissiveMasks_BrakeLight) {name=BrakeLightEmissiveMask, nodeColor=<rgba 0 0.1726 0.4792 1>}
  NamedRerouteDeclaration(None=@n17.EmissiveMasks_ReverseLightMask) {name=ReverseLightEmissiveMask, nodeColor=<rgba 0 0.1726 0.4792 1>}
  n20 = NamedRerouteUsage()
  n21 = Multiply(A=NamedRerouteUsage(), B=Multiply(A=@n20, B=NamedRerouteUsage())) {desc=HeadLight Color and Brightness}
  n22 = Multiply(A=Saturate(None=ConstantBiasScale(None=@n20) {bias=0.1, scale=1.5}), B=@n21) {desc="Enhances BC emissive so it doesn't fade to white.\r\nDoes slightly affect Headlight Color if it's enabled."}
  n23 = StaticSwitchParameter(True=Constant() {r=1}, False=Constant()) {desc=BC overrides emissive color while true., group=Vehicle_Chassis_Color, parameterName=UseBC_HeadlightEmissive}
  NamedRerouteDeclaration(None=Multiply(A=NamedRerouteUsage(), B=LinearInterpolate(A=@n21, B=@n22, Alpha=@n23) {desc="Push color of BC without\r\naffecting any other settings."}) {desc=HeadLight Variable Mask}) {name=HeadLightCombined, nodeColor=<rgba 1 1 1 1>}
  n24 = NamedRerouteUsage()
  n25 = Multiply(A=NamedRerouteUsage(), B=Multiply(A=@n24, B=NamedRerouteUsage())) {desc=HeadLight Color and Brightness}
  n26 = Multiply(A=Saturate(None=ConstantBiasScale(None=@n24) {bias=0.1, scale=1.5}), B=@n25) {desc="Enhances BC emissive so it doesn't fade to white.\r\nDoes slightly affect Headlight Color if it's enabled."}
  n27 = StaticSwitchParameter(True=Constant() {r=1}, False=Constant()) {desc=BC overrides emissive color while true., group=Vehicle_Chassis_Color, parameterName=UseBC_BrakeEmissive}
  NamedRerouteDeclaration(None=Multiply(A=NamedRerouteUsage(), B=LinearInterpolate(A=@n25, B=@n26, Alpha=@n27) {desc="Push color of BC without\r\naffecting any other settings."}) {desc=HeadLight Variable Mask}) {name=BrakeLightCombined, nodeColor=<rgba 1 1 1 1>}
  n28 = NamedRerouteUsage()
  n29 = Multiply(A=Constant() {r=1}, B=NamedRerouteUsage())
  NamedRerouteDeclaration(None=Multiply(A=NamedRerouteUsage(), B=Multiply(A=@n28, B=@n29) {desc=HeadLight Color and Brightness})) {name=ReverseLightCombined, nodeColor=<rgba 1 1 1 1>}
  n30 = NamedRerouteUsage()
  n31 = Multiply(A=NamedRerouteUsage(), B=NamedRerouteUsage())
  NamedRerouteDeclaration(None=Multiply(A=NamedRerouteUsage(), B=Multiply(A=@n30, B=@n31) {desc=HeadLight Color and Brightness})) {name=BoostCombined, nodeColor=<rgba 1 1 1 1>}
  NamedRerouteDeclaration(None=@n17.Masks_Parts) {name=ChassisEmissivePartsMask, nodeColor=<rgba 0 0.1726 0.4792 1>}
  NamedRerouteDeclaration(None=@n17.Color_Parts) {name=ChassisEmissivePartsColor, nodeColor=<rgba 0.9301 0 1 1>}
  n32 = Add(A=Add(A=NamedRerouteUsage(), B=NamedRerouteUsage()), B=NamedRerouteUsage())
  n33 = NamedRerouteUsage()
  NamedRerouteDeclaration(None=Add(A=Add(A=@n32, B=@n33), B=NamedRerouteUsage())) {name=EmissiveColorUpdated, nodeColor=<rgba 0 0.7656 1 1>}
  NamedRerouteDeclaration(None=@n19.ChassisEmissiveParts) {name=ChassisEmissivePartsBrightness, nodeColor=<rgba 0.0436 1 0 1>}
  n34 = NamedRerouteUsage()
  n35 = NamedRerouteUsage()
  NamedRerouteDeclaration(None=Multiply(A=NamedRerouteUsage(), B=Multiply(A=@n34, B=@n35))) {name=ChassisEmissivePartsCombined, nodeColor=<rgba 1 1 1 1>}
  n36 = NamedRerouteUsage()
  n37 = VectorParameter() {
   desc=BC overrides emissive color when set to 1
   group=Vehicle_Chassis_Emissive
   parameterName=UseBC_ReplaceEC
   channelNames {a=BoostEmissive, b=BrakeLightEmissive, g=HeadLightEmissive, r=ChassisEmissiveParts}
  }
  n38 = Floor(None=@n37.HeadLightEmissive)
  NamedRerouteDeclaration(None=LinearInterpolate(A=NamedRerouteUsage(), B=@n36, Alpha=Saturate(None=@n38))) {name=EC_HeadlightEmiss_Bool, nodeColor=<rgba 0.3763 0.7529 0.7157 1>}
  NamedRerouteDeclaration(None=@n17.Masks_Paint) {name=ChassisPaintedMask, nodeColor=<rgba 0 0.1726 0.4792 1>}
  n39 = Floor(None=@n37.ChassisEmissiveParts)
  NamedRerouteDeclaration(None=LinearInterpolate(A=NamedRerouteUsage(), B=@n36, Alpha=Saturate(None=@n39))) {name=EC_ChassisPartsEmis_Bool, nodeColor=<rgba 0.3763 0.7529 0.7157 1>}
  n40 = Floor(None=@n37.BrakeLightEmissive)
  NamedRerouteDeclaration(None=LinearInterpolate(A=NamedRerouteUsage(), B=@n36, Alpha=Saturate(None=@n40))) {name=EC_BrakeEmis_Bool, nodeColor=<rgba 0.3763 0.7529 0.7157 1>}
  n41 = Floor(None=@n37.BoostEmissive)
  NamedRerouteDeclaration(None=LinearInterpolate(A=NamedRerouteUsage(), B=@n36, Alpha=Saturate(None=@n41))) {name=EC_BoostEmis_Bool, nodeColor=<rgba 0.3763 0.7529 0.7157 1>}
  NamedRerouteDeclaration(None=@n17.Color_Brake) {name=ChassisBrakeColor, nodeColor=<rgba 0.9301 0 1 1>}
  n42 = NamedRerouteUsage()
  NamedRerouteDeclaration(None=StaticSwitchParameter(True=LinearInterpolate(A=@n12.EmissiveColor, B=@n12.EmissiveColor, Alpha=@n42), False=@n12.EmissiveColor) {group=02_Anime Parameters, parameterName=UsePBRSwitch, sortPriority=0}) {name=BaseColorTexture, desc=Leaving the name as-as for potential future work, nodeColor=<rgba 0.6038 0.3325 0.013 1>}
 ]}
