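# Material function that builds the Color_Lit and Color_Shaded outputs for anime-shaded vehicle chassis.
# Each color starts from a map or vector parameter, with a dynamic darkening option for shade.
# Switchable chains of masked lerps overlay colors on each, alongside MF_Vehicle_MixBCandEmissive blends, before cinematic tint functions produce the outputs.
# Per-channel parameters for parts, headlight, brake and boost set brightness and emissive-replaces-base-color flags, optionally only while active.
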MaterialFunction /VehicleCosmetics/Materials/VehicleParents_Version2/Material_Functions/AnimeShaderSupport/MF_Vehicle_ChassisAnime_Pre
graph {nodes [
  n1 = TextureSampleParameter2D(UVs=NamedRerouteUsage()) {group=01_Anime Maps, parameterName=Color_Shaded_Map, sortPriority=11, texture=/Game/Global/Textures/Default/Blanks/T_MidgreysRGB}
  n2 = NamedRerouteUsage()
  n3 = ScalarParameter() {defaultValue=0.5, group=01_Anime Maps, parameterName=Dynamic_ColorShaded_Darkening, sliderMax=1, sortPriority=11}
  n4 = VectorParameter() {group=01_Anime Maps, parameterName=Color_Shaded, sortPriority=11, defaultValue=<rgba 0.18 0.18 0.18 0>}
  n5 = StaticSwitchParameter(
   True=StaticSwitchParameter(True=@n1.RGB, False=Multiply(A=@n2, B=@n3)) {group=01_Anime Maps, parameterName=UseColorShadedMap, sortPriority=2}
   False=@n4.RGB
  ) {group=01_Anime Maps, parameterName=UseDynamicColorShaded, sortPriority=1}
  MaterialExpressionReroute_16 = Reroute(None=@n5)
  n6 = MaterialFunctionCall!(
   Color_Input=@MaterialExpressionReroute_16.Color_Input
   HeadlightColor=NamedRerouteUsage()
   Headlight_ECReplaceBC=NamedRerouteUsage()
   ChassisBrakeColor=NamedRerouteUsage()
   Brake_ECReplaceBC=NamedRerouteUsage()
   ChassisEmissivePartsColor=NamedRerouteUsage()
   Parts_ECReplaceBC=NamedRerouteUsage()
   BoostColor=NamedRerouteUsage()
   Boost_ECReplaceBC=NamedRerouteUsage()
   Headlight_EmissiveMask=NamedRerouteUsage()
   Brake_EmissiveMask=NamedRerouteUsage()
   Parts_EmissiveMask=NamedRerouteUsage()
   Boost_EmissiveMask=NamedRerouteUsage()
  ) {materialFunction=&/MF_Vehicle_MixBCandEmissive}
  MaterialExpressionReroute_17 = Reroute(None=@n6.EmissiveColors)
  n7 = LinearInterpolate(A=@n5, B=@MaterialExpressionReroute_17.EmissiveColors, Alpha=NamedRerouteUsage())
  Composite!(ColorShaded=@n5, ColorShaded_EmissiveBlend=@n7)
  Composite!(Color_Input=@n5)
  n8 = TextureSampleParameter2D(UVs=NamedRerouteUsage()) {group=01_Anime Maps, parameterName=Color_Lit_Map, sortPriority=10, texture=&T_WhitesRGB}
  n9 = StaticSwitchParameter(True=@n8.RGB, False=VectorParameter() {group=01_Anime Maps, parameterName=Color_Lit, sortPriority=10, defaultValue=<rgba 0.9 0.9 0.9 0>}.RGB) {group=01_Anime Maps, parameterName=UseColorLitMap, sortPriority=0}
  MaterialExpressionReroute_12 = Reroute(None=@n9)
  n10 = MaterialFunctionCall(
   Color_Input=@MaterialExpressionReroute_12.Color_Input
   HeadlightColor=NamedRerouteUsage()
   Headlight_ECReplaceBC=NamedRerouteUsage()
   ChassisBrakeColor=NamedRerouteUsage()
   Brake_ECReplaceBC=NamedRerouteUsage()
   ChassisEmissivePartsColor=NamedRerouteUsage()
   Parts_ECReplaceBC=NamedRerouteUsage()
   BoostColor=NamedRerouteUsage()
   Boost_ECReplaceBC=NamedRerouteUsage()
   Headlight_EmissiveMask=NamedRerouteUsage()
   Brake_EmissiveMask=NamedRerouteUsage()
   Parts_EmissiveMask=NamedRerouteUsage()
   Boost_EmissiveMask=NamedRerouteUsage()
  ) {materialFunction=&MF_Vehicle_MixBCandEmissive}
  MaterialExpressionReroute_15 = Reroute(None=@n10.EmissiveColors)
  n11 = LinearInterpolate(A=@n9, B=@MaterialExpressionReroute_15.EmissiveColors, Alpha=NamedRerouteUsage())
  Composite!(ColorLit=@n9, ColorLit_EmissiveBlend=@n11)
  Composite!(Color_Input=@n9)
  MaterialExpressionReroute_23 = Reroute(None=@n9)
  n12 = Multiply(A=NamedRerouteUsage()) {constB=10}
  n13 = LinearInterpolate(
   A=@MaterialExpressionReroute_23.ColorLit
   B=Multiply(A=NamedRerouteUsage(), B=NamedRerouteUsage())
   Alpha=Multiply(A=Saturate(None=@n12), B=NamedRerouteUsage())
  )
  n14 = Multiply(A=NamedRerouteUsage(), B=NamedRerouteUsage())
  n15 = NamedRerouteUsage()
  n16 = Multiply(A=Saturate(None=Multiply(A=@n15) {constB=10}), B=NamedRerouteUsage())
  n17 = NamedRerouteUsage()
  n18 = NamedRerouteUsage()
  n19 = Saturate(None=Multiply(A=NamedRerouteUsage()) {constB=10})
  n20 = NamedRerouteUsage()
  n21 = Multiply(A=NamedRerouteUsage()) {constB=10}
  n22 = LinearInterpolate(
   A=LinearInterpolate(A=LinearInterpolate(A=@n13, B=@n14, Alpha=@n16), B=Multiply(A=@n17, B=@n18), Alpha=Multiply(A=@n19, B=@n20))
   B=Multiply(A=NamedRerouteUsage(), B=NamedRerouteUsage())
   Alpha=Multiply(A=Saturate(None=@n21), B=NamedRerouteUsage())
  )
  MaterialExpressionReroute_25 = Reroute(None=@n22)
  n23 = StaticSwitch(True=@MaterialExpressionReroute_25.ColorLitFinal, False=@n9, Value=NamedRerouteUsage())
  FunctionOutput(None=NamedRerouteDeclaration(None=MaterialFunctionCall!(LitBaseColorIn=@n23) {materialFunction=/Game/Packages/Fortress_SharedMaterials/Cinematics/Anime/MF_CINE_LitDiffuseTint}.Result) {name=Color_Lit, nodeColor=<rgba 0 0.4609 1 1>}) {outputName=Color_Lit, sortPriority=11}
  MaterialExpressionReroute_38 = Reroute(None=@n5)
  n24 = Multiply(A=NamedRerouteUsage()) {constB=10}
  n25 = LinearInterpolate(
   A=@MaterialExpressionReroute_38.ColorShaded
   B=Multiply(A=NamedRerouteUsage(), B=NamedRerouteUsage())
   Alpha=Multiply(A=Saturate(None=@n24), B=NamedRerouteUsage())
  )
  n26 = Multiply(A=NamedRerouteUsage(), B=NamedRerouteUsage())
  n27 = NamedRerouteUsage()
  n28 = Multiply(A=Saturate(None=Multiply(A=@n27) {constB=10}), B=NamedRerouteUsage())
  n29 = NamedRerouteUsage()
  n30 = NamedRerouteUsage()
  n31 = Saturate(None=Multiply(A=NamedRerouteUsage()) {constB=10})
  n32 = NamedRerouteUsage()
  n33 = Multiply(A=NamedRerouteUsage()) {constB=10}
  n34 = LinearInterpolate(
   A=LinearInterpolate(A=LinearInterpolate(A=@n25, B=@n26, Alpha=@n28), B=Multiply(A=@n29, B=@n30), Alpha=Multiply(A=@n31, B=@n32))
   B=Multiply(A=NamedRerouteUsage(), B=NamedRerouteUsage())
   Alpha=Multiply(A=Saturate(None=@n33), B=NamedRerouteUsage())
  )
  MaterialExpressionReroute_42 = Reroute(None=@n34)
  n35 = NamedRerouteUsage()
  FunctionOutput(None=MaterialFunctionCall!(ShadedBaseColorIn=StaticSwitch(True=@MaterialExpressionReroute_42.ColorShadedFinal, False=@n5, Value=@n35)) {materialFunction=/Game/Packages/Fortress_SharedMaterials/Cinematics/Anime/MF_CINE_ShadedDiffuseTint}.Result) {outputName=Color_Shaded, sortPriority=12}
  NamedRerouteDeclaration(None=FunctionInput(Preview=TextureCoordinate()) {bUsePreviewValueAsDefault=true, inputName=CustomUV, inputType=FunctionInput_Vector2, sortPriority=5}) {name=UV Coords, nodeColor=<rgba 0.5 0.5 0 1>}
  NamedRerouteDeclaration(None=FunctionInput() {inputName="Color_Shaded (Passthrough)", sortPriority=11}) {name=Passthrough_ColorShaded, nodeColor=<rgba 0 1 0.3203 1>}
  n36 = VectorParameter() {
   desc=Brightness Settings
   group=Vehicle_Chassis_Color
   parameterName=ECtoBC_ColorShadedBrightness
   sortPriority=0
   channelNames {a=BoostEmissive, b=BrakeLightEmissive, g=HeadLightEmissive, r=ChassisEmissiveParts}
   defaultValue=<rgba 0.5 0.5 0.5 0.5>
  }
  NamedRerouteDeclaration(None=@n36.HeadLightEmissive) {name=ECtoBC_HeadlightsShadedBright, nodeColor=<rgba 0 1 0.2734 1>}
  NamedRerouteDeclaration(None=@n36.BrakeLightEmissive) {name=ECtoBC_BrakeShadedBright, nodeColor=<rgba 0 1 0.2734 1>}
  NamedRerouteDeclaration(None=@n36.BoostEmissive) {name=ECtoBC_BoostShadedBright, nodeColor=<rgba 0 1 0.2734 1>}
  n37 = Max(A=NamedRerouteUsage(), B=NamedRerouteUsage())
  n38 = NamedRerouteUsage()
  n39 = Max(A=Max(A=Max(A=@n37, B=@n38), B=NamedRerouteUsage()), B=NamedRerouteUsage())
  NamedRerouteDeclaration(None=Saturate(None=ConstantBiasScale(None=@n39) {bias=0, scale=2})) {name=EmissiveMasks, nodeColor=<rgba 0.4297 0 1 1>}
  MaterialExpressionReroute_24 = Reroute(None=@n11)
  NamedRerouteDeclaration(None=@MaterialExpressionReroute_24.ColorLit_EmissiveBlend) {name=ColorLit_EmissiveBlend, nodeColor=<rgba 0.2188 0 1 1>}
  n40 = VectorParameter() {
   desc=Brightness Settings
   group=Vehicle_Chassis_Color
   parameterName=ECtoBC_ColorLitBrightness
   sortPriority=0
   channelNames {a=BoostEmissive, b=BrakeLightEmissive, g=HeadLightEmissive, r=ChassisEmissiveParts}
   defaultValue=<rgba 0.85 0.85 0.85 0.85>
  }
  NamedRerouteDeclaration(None=@n40.ChassisEmissiveParts) {name=ECtoBC_PartsLitBright, nodeColor=<rgba 0 1 0.2734 1>}
  MaterialExpressionReroute_40 = Reroute(None=@n7)
  NamedRerouteDeclaration(None=@MaterialExpressionReroute_40.ColorShaded_EmissiveBlend) {name=ColorShaded_EmissiveBlend, nodeColor=<rgba 0.3594 1 0 1>}
  n41 = MaterialFunctionCall() {materialFunction=&/MF_Vehicle_ChassisAnimeVariables}
  NamedRerouteDeclaration(None=@n41.Color_Headlight) {name=HeadlightColor, nodeColor=<rgba 0.9282 0 1 1>}
  NamedRerouteDeclaration(None=@n41.Color_Boost) {name=BoostColor, nodeColor=<rgba 0.9282 0 1 1>}
  NamedRerouteDeclaration(None=@n40.HeadLightEmissive) {name=ECtoBC_HeadlightsLitBright, nodeColor=<rgba 0 1 0.2734 1>}
  n42 = VectorParameter() {
   desc=Brightness Settings
   group=Vehicle_Chassis_Emissive
   parameterName=Chassis_EmissiveBrightness
   sortPriority=0
   channelNames {a=BoostEmissive, b=BrakeLightEmissive, g=HeadLightEmissive, r=ChassisEmissiveParts}
   defaultValue=<rgba 1 1 1 1>
  }
  NamedRerouteDeclaration(None=@n42.HeadLightEmissive) {name=HeadlightEmissiveBrightness, nodeColor=<rgba 0.0436 1 0 1>}
  NamedRerouteDeclaration(None=@n42.BrakeLightEmissive) {name=BrakelightEmissiveBrightness, nodeColor=<rgba 0.0436 1 0 1>}
  NamedRerouteDeclaration(None=@n42.BoostEmissive) {name=BoostEmissiveBrightness, nodeColor=<rgba 0.0436 1 0 1>}
  NamedRerouteDeclaration(None=@n41.BPVariable_BoostGlowIntensity) {name=BoostGlowVariable, nodeColor=<rgba 0 0 0 1>}
  NamedRerouteDeclaration(None=@n41.BPVariable_isBraking) {name=isBrakingVariable, nodeColor=<rgba 0 0 0 1>}
  NamedRerouteDeclaration(None=@n40.BrakeLightEmissive) {name=ECtoBC_BrakeLitBright, nodeColor=<rgba 0 1 0.2734 1>}
  NamedRerouteDeclaration(None=@n40.BoostEmissive) {name=ECtoBC_BoostLitBright, nodeColor=<rgba 0 1 0.2734 1>}
  NamedRerouteDeclaration(None=@n41.BPVariable_ReverseLight) {name=ReverseLightVariable, nodeColor=<rgba 0 0 0 1>}
  NamedRerouteDeclaration(None=@n41.BPVariable_HeadlightsFullyOn) {name=HeadlightsFullyOnVariable, nodeColor=<rgba 0 0 0 1>}
  NamedRerouteDeclaration(None=@n41.Masks_Boost) {name=ChassisBoostEmissiveMask, nodeColor=<rgba 0 0.1726 0.4792 1>}
  NamedRerouteDeclaration(None=@n41.EmissiveMasks_Headlight) {name=HeadlightEmissiveMask, nodeColor=<rgba 0 0.1726 0.4792 1>}
  NamedRerouteDeclaration(None=@n41.EmissiveMasks_BrakeLight) {name=BrakeLightEmissiveMask, nodeColor=<rgba 0 0.1726 0.4792 1>}
  NamedRerouteDeclaration(None=@n41.EmissiveMasks_ReverseLightMask) {name=ReverseLightEmissiveMask, nodeColor=<rgba 0 0.1726 0.4792 1>}
  NamedRerouteDeclaration(None=@n36.ChassisEmissiveParts) {name=ECtoBC_PartsShadedBright, nodeColor=<rgba 0 1 0.2734 1>}
  NamedRerouteDeclaration(None=@n41.Masks_Parts) {name=ChassisEmissivePartsMask, nodeColor=<rgba 0 0.1726 0.4792 1>}
  NamedRerouteDeclaration(None=@n41.Color_Parts) {name=ChassisEmissivePartsColor, nodeColor=<rgba 0.9301 0 1 1>}
  NamedRerouteDeclaration(None=@n42.ChassisEmissiveParts) {name=ChassisEmissivePartsBrightness, nodeColor=<rgba 0.0436 1 0 1>}
  NamedRerouteDeclaration(None=@n41.Masks_Paint) {name=ChassisPaintedMask, nodeColor=<rgba 0 0.1726 0.4792 1>}
  NamedRerouteDeclaration(None=@n41.Color_Brake) {name=ChassisBrakeColor, nodeColor=<rgba 0.9301 0 1 1>}
  n43 = VectorParameter() {
   desc=Emissive color overrides base color when set to 1
   group=Vehicle_Chassis_Color
   parameterName=UseEC_ReplaceBC
   channelNames {a=BoostEmissive, b=BrakeLightEmissive, g=HeadLightEmissive, r=ChassisEmissiveParts}
  }
  n44 = Saturate(None=Floor(None=@n43.ChassisEmissiveParts))
  n45 = Saturate(None=NamedRerouteUsage())
  n46 = VectorParameter() {
   desc="When set to 1, the emissive color will only replace the base color if the emissive group is active (e.g., the headlight is on)."
   group=Vehicle_Chassis_Color
   parameterName=UseEC_ReplaceOnlyWhenActive
   sortPriority=33
   channelNames {a=BoostEmissive, b=BrakeLightEmissive, g=HeadLightEmissive, r=ChassisEmissiveParts}
   defaultValue=<rgba 1 1 1 0>
  }
  n47 = Floor(None=@n46.ChassisEmissiveParts)
  NamedRerouteDeclaration(None=LinearInterpolate(A=@n44, B=Multiply(A=@n45, B=@n44), Alpha=Saturate(None=@n47))) {name=EC_ReplaceBC_EmissiveParts, nodeColor=<rgba 0.875 0 1 1>}
  n48 = Floor(None=@n43.HeadLightEmissive)
  n49 = Floor(None=@n46.HeadLightEmissive)
  NamedRerouteDeclaration(None=LinearInterpolate(A=Saturate(None=@n48), B=NamedRerouteUsage(), Alpha=Saturate(None=@n49))) {name=EC_ReplaceBC_HeadLight, nodeColor=<rgba 0.875 0 1 1>}
  n50 = Floor(None=@n43.BoostEmissive)
  n51 = Floor(None=@n46.BoostEmissive)
  NamedRerouteDeclaration(None=LinearInterpolate(A=Saturate(None=@n50), B=NamedRerouteUsage(), Alpha=Saturate(None=@n51))) {name=EC_ReplaceBC_Boost, nodeColor=<rgba 0.875 0 1 1>}
  n52 = Floor(None=@n43.BrakeLightEmissive)
  n53 = Floor(None=@n46.BrakeLightEmissive)
  NamedRerouteDeclaration(None=LinearInterpolate(A=Saturate(None=@n52), B=NamedRerouteUsage(), Alpha=Saturate(None=@n53))) {name=EC_ReplaceBC_BrakeLight, nodeColor=<rgba 0.875 0 1 1>}
  NamedRerouteDeclaration(None=StaticBoolParameter() {group=Vehicle_ShaderControls, parameterName=UseChassisSettings, sortPriority=30}) {name=UseChassisBool, nodeColor=<rgba 0.6406 0 1 1>}
  NamedRerouteDeclaration(None=FunctionInput() {inputName="Color_Lit (Passthrough)", sortPriority=10}) {name=Passthrough_ColorLit, nodeColor=<rgba 0 1 0.3203 1>}
  PinBase() {reroutePins [
    ColorLit {expression=&self:MaterialExpressionReroute_23}
    ColorLit_EmissiveBlend {expression=&self:MaterialExpressionReroute_24}
   ]}
  PinBase!(ColorLitFinal=@n22) {reroutePins [ColorLitFinal {expression=&self:MaterialExpressionReroute_25}]}
  PinBase!() {reroutePins [Color_Input {expression=&self:MaterialExpressionReroute_12}]}
  PinBase!(EmissiveColors=@n10.EmissiveColors) {reroutePins [EmissiveColors {expression=&self:MaterialExpressionReroute_15}]}
  PinBase!() {reroutePins [
    ColorShaded {expression=&self:MaterialExpressionReroute_38}
    ColorShaded_EmissiveBlend {expression=&self:MaterialExpressionReroute_40}
   ]}
  PinBase!(ColorShadedFinal=@n34) {reroutePins [ColorShadedFinal {expression=&self:MaterialExpressionReroute_42}]}
  PinBase!() {reroutePins [Color_Input {expression=&self:MaterialExpressionReroute_16}]}
  PinBase!(EmissiveColors=@n6.EmissiveColors) {reroutePins [EmissiveColors {expression=&self:MaterialExpressionReroute_17}]}
 ]}
