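# Parent material for anime-shaded vehicle cosmetics, built on the shared 2023 character anime shader with material attributes and the Cloth shading model.
# Character Pre feeds Calculations, while a vehicle chassis Pre function processes its lit and shaded colors.
# Character Post combines those colors with the computed gradients and surface inputs.
# The UseInteriorSettings static switch outputs that result directly; otherwise it passes through vehicle ChassisAnime_Post and Anime_FNPostFX functions.
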Material /VehicleCosmetics/Materials/VehicleParents_Version2/MAT_Vehicle_AnimeShader
graph {
 properties {
  bAutomaticallySetUsageInEditor=false
  bUseMaterialAttributes=true
  bUsedWithClothing=true
  bUsedWithMeshDeformer=true
  bUsedWithMorphTargets=true
  bUsedWithSkeletalMesh=true
  bUsedWithStaticLighting=true
  pixelDepthOffsetMode=PDOM_Legacy
  shadingModel=MSM_Cloth
  usedShadingModels=Cloth
 }
 nodes [
  n1 = MaterialFunctionCall!(CustomUV) {materialFunction=/Game/Global/Materials/Character_Parent_2023/Anime_Parent_2023/MaterialFunctions/MFC_AnimeShader_2023_Pre}
  n2 = MaterialFunctionCall!(CustomLightVector, OverrideNormal_WS, STT=@n1.STT, Normal_Map=@n1.STT, DFL=@n1.STT, FX=@n1.STT) {materialFunction=/Game/Global/Materials/Character_Parent_2023/Anime_Parent_2023/MaterialFunctions/MFC_AnimeShader_2023_Calculations}
  n3 = MaterialFunctionCall!(CustomUV, "Color_Lit (Passthrough)"=@n1.Color_Lit, "Color_Shaded (Passthrough)"=@n1.Color_Lit) {materialFunction=/VehicleCosmetics/Materials/VehicleParents_Version2/Material_Functions/AnimeShaderSupport/MF_Vehicle_ChassisAnime_Pre}
  n4 = MaterialFunctionCall!(
   Shading_Gradient=@n2.Shading_Gradient
   Threshold_Gradient=@n2.Shading_Gradient
   Specular_Gradient=@n2.Shading_Gradient
   Sheen_Mask=@n2.Shading_Gradient
   Thickness=@n2.Shading_Gradient
   DFL_Mask=@n2.Shading_Gradient
   FX=@n2.Shading_Gradient
   Metallic_Gradient=@n2.Shading_Gradient
   RimLight_Gradient=@n2.Shading_Gradient
   NormalBuffer=@n2.Shading_Gradient
   --------
   CINE_LightSources=@n2.Shading_Gradient
   CINE_LightColor=@n2.Shading_Gradient
   ---------
   Color_Lit=@n3.Color_Lit
   Color_Shaded=@n3.Color_Lit
   SRM=@n1.SRM
   Emissive=@n1.SRM
   DFL_Color=@n1.SRM
   Specular_Color=@n1.SRM
   RimLight_Color=@n1.SRM
  ) {materialFunction=/Game/Global/Materials/Character_Parent_2023/Anime_Parent_2023/MaterialFunctions/MFC_AnimeShader_2023_Post}
  StaticSwitch(
   True=@n4.Result
   False=MaterialFunctionCall!(In=MaterialFunctionCall!(Input=@n4.Result) {materialFunction=/VehicleCosmetics/Materials/VehicleParents_Version2/Material_Functions/AnimeShaderSupport/MF_Vehicle_ChassisAnime_Post}.Result) {materialFunction=/VehicleCosmetics/Materials/VehicleParents_Version2/Material_Functions/AnimeShaderSupport/MF_Vehicle_Anime_FNPostFX}.Result
   Value=StaticBoolParameter() {group=Vehicle_ShaderControls, parameterName=UseInteriorSettings, sortPriority=30}
  )
 ]
}
