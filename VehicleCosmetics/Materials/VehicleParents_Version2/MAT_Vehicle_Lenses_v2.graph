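# Translucent parent material for vehicle lenses, using Thin Translucent shading with per-pixel surface lighting.
# MF_Lenses_Base_v2 builds the material attributes, with static switches for custom rearlights, reflectors, miscellaneous lenses, shared headlights and translucency sorting.
# Its result passes through MF_Vehicle_FNPostFX, and its transmittance color also drives both thin-translucent transmittance and surface coverage.
# Damage and de-rez transition functions are present but not fed by the lens result.
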Material /VehicleCosmetics/Materials/VehicleParents_Version2/MAT_Vehicle_Lenses_v2
graph {
 properties {
  allowTranslucentCustomDepthWrites=true
  bAutomaticallySetUsageInEditor=false
  bUseMaterialAttributes=true
  bUsedWithMeshDeformer=true
  bUsedWithSkeletalMesh=true
  blendMode=BLEND_Translucent
  refractionMethod=RM_IndexOfRefraction
  shadingModel=MSM_ThinTranslucent
  translucencyLightingMode=TLM_SurfacePerPixelLighting
  usedShadingModels=Thin Translucent
 }
 nodes [
  MaterialFunctionCall(In=MaterialFunctionCall() {materialFunction=&MF_BaseVehicle_Damage}.Result) {materialFunction=&MF_VehicleDeRez_Transition}
  n1 = MaterialFunctionCall!(
   UseCustomRearlights=StaticBoolParameter() {group=Lenses, parameterName=UseCustomRearlights}
   UseCustomReflector=StaticBoolParameter() {group=Lenses, parameterName=UseCustomReflectors}
   UseCustomSpecial=StaticBoolParameter() {group=Lenses, parameterName=UseCustomMiscellaneous}
   UseSortTranslucency=StaticBoolParameter() {group=Lenses, parameterName=UseSortTranslucency}
   UseHeadlightsAsShared=StaticBoolParameter() {defaultValue=true, group=Lenses, parameterName=UseHeadlightsAsShared}
  ) {materialFunction=/VehicleCosmetics/Materials/VehicleParents_Version2/Material_Functions/MF_Lenses_Base_v2}
  MaterialFunctionCall!(In=@n1.Result) {materialFunction=/VehicleCosmetics/Materials/MaterialFunctions/MF_Vehicle_FNPostFX}
  n2 = GetMaterialAttributes(None=@n1.Result) {attributeGetTypes [#1, #2]}
  ThinTranslucentMaterialOutput!(TransmittanceColor=@n2.TransmittanceColor, SurfaceCoverage=@n2.TransmittanceColor)
 ]
}
