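# Parent vehicle material for a stainless anodized finish on static or skeletal meshes, using material attributes with Clear Coat shading.
# MF_Stainless_Anodized builds the surface, then MF_BodyNormalLOD processes it before MF_BaseVehicle_Damage applies damage.
# Get/SetMaterialAttributes bypasses damage to rewrite the Shading Model, Anisotropy and Tangent slots from pre-damage attributes.
# MF_VehicleDeRez_Transition runs last on the combined result.
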Material /VehicleCosmetics/Materials/VehicleParents_Version2/MAT_Vehicle_Stainless_Anodized
graph {
 properties {
  bUseMaterialAttributes=true
  bUsedWithMeshDeformer=true
  bUsedWithSkeletalMesh=true
  pixelDepthOffsetMode=PDOM_Legacy
  shadingModel=MSM_FromMaterialExpression
  usedShadingModels=Clear Coat
 }
 nodes [
  n1 = GetMaterialAttributes(None=MaterialFunctionCall!(In=MaterialFunctionCall() {materialFunction=/VehicleCosmetics/Materials/VehicleParents_Version2/Material_Functions/MF_Stainless_Anodized}.Result) {materialFunction=/VehicleCosmetics/Materials/MaterialFunctions/MF_BodyNormalLOD}.Result) {attributeGetTypes [#1, #2, #3]}
  MaterialFunctionCall(In=SetMaterialAttributes!(
    MaterialAttributes=MaterialFunctionCall(Material=@n1.MaterialAttributes) {materialFunction=&MF_BaseVehicle_Damage}.Result
    Shading Model=@n1.Shading Model
    Anisotropy=@n1.Shading Model
    Tangent=@n1.Shading Model
   ) {attributeSetTypes [#1, #2, #3]}) {materialFunction=&MF_VehicleDeRez_Transition}
 ]
}
